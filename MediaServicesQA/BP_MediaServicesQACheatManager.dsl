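; QA cheat manager that provides commands for testing streamed video playback and decoding media IDs into logged video URLs.
; Stream commands run one Fortnite video widget at a time, removed on deactivation.
; Simple player commands spawn multiple test players, each mapped to a bordered tile in a top-layer StackBox and removed on close.
; The misleadingly named HandleStreamVideoSkipped builds those tiles, displaying each MediaTexture through a dynamic UI material.
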(fn AddSimpleVideoPlayerById
  (params (MediaId :type String) (Priority :type ETestSimplePlayerPriority%20Enum :default "NewEnumerator0"))
  (bind _returnvalue (Game|ConstructObjectfromClass "/MediaServicesQA/Assets/BP_TestSimplePlayer.BP_TestSimplePlayer_C"))
  (bind _result (Default|Create _returnvalue Priority true false :!class /MediaServicesQA/Assets/BP_TestSimplePlayer.BP_TestSimplePlayer_C))
  (if _result
    (Default|BindEventtoOnClose _returnvalue (EventDispatchers|CreateEvent :!event Handle%20Player%20Closed))
    (bind _result_2 (Private|HandleStreamVideoSkipped (Variables|/MediaServicesQA/Assets/BP_TestSimplePlayer.BP_TestSimplePlayer_C|GetMediaTexture _returnvalue)))
    (Utilities|Map|Add (Variables|Default|GetSimplePlayersToWidget) _returnvalue _result_2)
    (Development|PrintString (Utilities|String|Append "Created Player: " (Guid|ToString(Guid) (Class|BPTestSimplePlayer|GetId _returnvalue))))
    (Default|StartbyId _returnvalue (Private|GetMediaId MediaId) :!class /MediaServicesQA/Assets/BP_TestSimplePlayer.BP_TestSimplePlayer_C)))

(fn DecodeMediaId
  (params (MediaId :type String) (bEnableBLURLRetries :type Boolean))
  (bind _returnvalue (Game|ConstructObjectfromClass "/Script/FortniteGame.FortStreamMediaSource"))
  (Variables|/Script/EpicStreamMediaSource.EpicStreamMediaSource|SetStreamID MediaId _returnvalue)
  (Variables|/Script/EpicStreamMediaSource.EpicStreamMediaSource|SetbEnableBLURLRetries bEnableBLURLRetries _returnvalue)
  (Stream|BindEventtoOnVideoUrlSuccess _returnvalue (EventDispatchers|CreateEvent :!event Handle%20Video%20UrlSuccess))
  (Stream|BindEventtoOnVideoUrlFailed _returnvalue (EventDispatchers|CreateEvent :!event Handle%20Video%20UrlFail))
  (Stream|RequestVideoUrl _returnvalue (Fortnite|GetFirstLocalFortPlayerController true)))

(fn GetMediaId
  (params (MediaId :type String))
  (if (Utilities|String|IsEmpty MediaId)
    (return "yZCScFFPeBHXNMtusu")
    (else
      (return MediaId))))

(fn HandlePlayerClosed
  (params (Player :type BP%20Test%20Simple%20Player%20Object%20Reference))
  (bind _simpleplayerstowidget (Variables|Default|GetSimplePlayersToWidget))
  (bind (_value _returnvalue) (Utilities|Map|Find _simpleplayerstowidget Player))
  (if _returnvalue
    (bind _output_get (LocalVariables|SetLocalWidget _value))
    (Utilities|Map|Remove _simpleplayerstowidget Player)
    (Widget|RemovefromParent _output_get)))

(fn HandleVideoUrlFail
  (params (RetrievedUrl :type String) (Code :type Integer) (RequestType :type EEpicMediaRequestType%20Enum :default "UNKNOWN"))
  (Utilities|String|LogString (Utilities|String|Append "Decoded VideoURL: " "false\\n" RetrievedUrl)))

(fn HandleVideoUrlSuccess
  (params (RetrievedUrl :type String) (Code :type Integer) (RequestType :type EEpicMediaRequestType%20Enum :default "UNKNOWN"))
  (Utilities|String|LogString (Utilities|String|Append "Decoded VideoURL: " "true\\n" RetrievedUrl)))

(fn HandleOnWidgetDeactivated
  (Exec|RemoveStreamVideo))

(fn HandleStreamVideoSkipped
  (params (MediaTexture :type Media%20Texture%20Object%20Reference))
  (Utilities|IsValid (Variables|Default|GetSimpleMediaPlayerStackBoxWidget)
    (:Is%20Valid
      _)
    (:Is%20Not%20Valid
      (bind _returnvalue (Game|ConstructObjectfromClass "/Script/UMG.StackBox"))
      (bind _output_get (Variables|Default|SetSimpleMediaPlayerStackBoxWidget _returnvalue))
      (Variables|/Script/UMG.StackBox|SetOrientation "Orient_Horizontal" _output_get)
      (UserInterface|AddWidget (EngineSubsystems|GetGameViewportSubsystem) _output_get (struct :Anchors (struct :Minimum "(X=0.050000,Y=0.050000)" :Maximum "(X=0.950000,Y=0.950000)") :Offsets "(Left=0.000000,Top=0.000000,Right=0.000000,Bottom=0.000000)" :Alignment "(X=0.000000,Y=0.000000)" :ZOrder 999999999 :bAutoRemoveOnWorldRemoved true))))
  (Utilities|FlowControl|Sequence
    (:then_0
      (bind _returnvalue_2 (Game|ConstructObjectfromClass "/Script/UMG.Border"))
      (Variables|/Script/UMG.Border|SetBrushColor "(R=0.354167,G=0.354167,B=0.354167,A=1.000000)" _returnvalue_2)
      (bind _returnvalue_3 (Widget|Panel|AddChild (Variables|Default|GetSimpleMediaPlayerStackBoxWidget) _returnvalue_2))
      (Variables|/Script/UMG.StackBoxSlot|SetVerticalAlignment "VAlign_Center" (Utilities|Casting|CastToStackBoxSlot _returnvalue_3))
      (Variables|/Script/UMG.StackBoxSlot|SetHorizontalAlignment "HAlign_Center" (Utilities|Casting|CastToStackBoxSlot _returnvalue_3)))
    (:then_1
      (bind _returnvalue_4 (Game|ConstructObjectfromClass "/Script/UMG.Image"))
      (Widget|Panel|AddChild _returnvalue_2 _returnvalue_4)
      (Appearance|SetBrush _returnvalue_4 (struct :TintColor "(SpecifiedColor=(R=1.000000,G=1.000000,B=1.000000,A=1.000000),ColorUseRule=UseColor_Specified)" :DrawAs "Image" :Tiling "NoTile" :Mirroring "NoMirror" :ImageSize (struct :X 6.4e+02 :Y 4.8e+02) :Margin "(Left=0.000000,Top=0.000000,Right=0.000000,Bottom=0.000000)" :ResourceObject 0 :OutlineSettings "(CornerRadii=(X=0.000000,Y=0.000000,Z=0.000000,W=0.000000),Color=(SpecifiedColor=(R=0.000000,G=0.000000,B=0.000000,A=0.000000),ColorUseRule=UseColor_Specified),Width=0.000000,RoundingType=HalfHeightRadius,bUseBrushTransparency=False)") :!class /Script/UMG.Image)
      (bind _returnvalue_5 (Rendering|Material|CreateDynamicMaterialInstance "/MediaServicesQA/Assets/M_TextureUserInterface.M_TextureUserInterface" "" "Transient" :!class /Script/Engine.KismetMaterialLibrary))
      (Rendering|Material|SetTextureParameterValue _returnvalue_5 "MediaTexture" MediaTexture)
      (Appearance|SetBrushfromMaterial _returnvalue_4 _returnvalue_5 :!class /Script/UMG.Image)
      (return _returnvalue_2))))

(fn RemoveAllSimpleVideoPlayers
  (bind _keys (Utilities|Map|Keys (Variables|Default|GetSimplePlayersToWidget)))
  (for _array_element _keys
    (Class|BPTestSimplePlayer|Close _array_element)))

(fn RemoveSimpleVideoPlayer
  (params (Id :type String))
  (bind (_outguid _success) (Guid|ParseStringtoGuid Id))
  (if _success
    (bind _keys (Utilities|Map|Keys (Variables|Default|GetSimplePlayersToWidget)))
    (for _array_element _keys
      (if (== _outguid (Class|BPTestSimplePlayer|GetId _array_element))
        (Class|BPTestSimplePlayer|Close _array_element)))))

(fn RemoveStreamVideo
  (bind _teststreamvideowidget (Variables|Default|GetTestStreamVideoWidget))
  (Utilities|IsValid _teststreamvideowidget
    (:Is%20Valid
      (Widget|RemovefromParent _teststreamvideowidget)
      (Variables|Default|SetTestStreamVideoWidget 0))))

(fn StartStreamVideoById
  (params (MediaId :type String) (bAllowSkipping :type Boolean :default true) (bLoop :type Boolean) (bAudio :type Boolean :default true))
  (bind _config (LocalVariables|GetConfig))
  (Exec|RemoveStreamVideo)
  (bind _returnvalue (UserInterface|CreateWidget "/Game/UI/Foundation/Movie/B_FortVideoWidgetBase.B_FortVideoWidgetBase_C" (Game|GetPlayerController 0)))
  (bind _output_get (Variables|Default|SetTestStreamVideoWidget _returnvalue))
  (Utilities|FlowControl|Sequence
    (:then_0
      (UserInterface|Viewport|AddtoViewport _output_get 0))
    (:then_1
      (bind _returnvalue_2 (Game|ConstructObjectfromClass "/Script/FortniteUI.FortMediaVideoWidgetConfig"))
      (LocalVariables|SetConfig _returnvalue_2))
    (:then_2
      (Variables|/Script/EpicMediaVideoWidgetBase.EpicMediaVideoWidgetConfig|SetMediaID MediaId _config)
      (Variables|/Script/EpicMediaVideoWidgetBase.EpicMediaVideoWidgetConfig|SetMediaName "TestStreamedVideo" _config)
      (Variables|/Script/EpicMediaVideoWidgetBase.EpicMediaVideoWidgetConfig|SetbAutoPlay true _config)
      (Variables|/Script/EpicMediaVideoWidgetBase.EpicMediaVideoWidgetConfig|SetbAllowSkipping bAllowSkipping _config)
      (Variables|/Script/EpicMediaVideoWidgetBase.EpicMediaVideoWidgetConfig|SetbHideControls false _config)
      (Variables|/Script/EpicMediaVideoWidgetBase.EpicMediaVideoWidgetConfig|SetbLoop bLoop _config)
      (Variables|/Script/EpicMediaVideoWidgetBase.EpicMediaVideoWidgetConfig|SetbNoAudio (not bAudio) _config))
    (:then_3
      (bind _returnvalue_3 (Media|SetConfig _output_get (LocalVariables|GetConfig)))
      (if _returnvalue_3
        (bind _returnvalue_4 (Media|Init _output_get))
        (if _returnvalue_4
          (ActivatableWidget|ActivateWidget _output_get)
          (else
            (Development|PrintString "Init failed" true true "(R=1.000000,G=0.000000,B=0.000000,A=1.000000)")))
        (else
          (Development|PrintString "Init failed" true true "(R=1.000000,G=0.000000,B=0.000000,A=1.000000)"))))
    (:then_4
      (Events|BindEventtoOnWidgetDeactivated _output_get (EventDispatchers|CreateEvent :!event HandleOnWidgetDeactivated)))))

(fn TestStreamedVideo
  (params (MediaId :type String))
  (if (Utilities|String|IsEmpty MediaId)
    (Exec|RemoveStreamVideo)
    (else
      (Exec|StartStreamVideobyId MediaId))))
